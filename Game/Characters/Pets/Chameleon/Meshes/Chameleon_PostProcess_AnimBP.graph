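# Post-process AnimBP for the Chameleon pet, adding procedural neck and head motion and tongue constraints over the input pose.
# When bPetWantsAnimDynamics is set, emote/DBNO state selects sprung neck_01/head AnimDynamics; otherwise neck_02/head dynamics scale with pawn speed.
# Without that flag, the cached input pose bypasses dynamics.
# Finally, chained position, then scale constraints make tongue1–4 each follow the preceding bone, starting from tongue.
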AnimBlueprint /Game/Characters/Pets/Chameleon/Meshes/Chameleon_PostProcess_AnimBP
target_skeleton=&/Chameleon_Skeleton
graphs [
 AnimGraph {nodes [
   n1 = Get PawnSpeedForAnimDynamics() -> (PawnSpeedForAnimDynamics:float)
   n2 = Anim Dynamics - Bone: neck_02(ComponentPose=Local To Component(LocalPose=Use cached pose 'Sub Graph Cache'()), Alpha=@n1, ExternalForce="0,0,0") {
    angularDampingOverride=0.999
    angularSpringConstant=200
    bGravityOverrideInSimSpace=true
    bOverrideAngularDamping=true
    lODThreshold=1
    numSolverIterationsPostUpdate=4
    numSolverIterationsPreUpdate=16
    alphaBoolBlend {blendOption=Linear}
    alphaScaleBiasClamp {bInterpResult=true, bMapRange=true, inRange {max=500}}
    boundBone {boneName=neck_02}
    componentLinearAccScale=<0.5 0.5 0.5>
    physicsBodyDefinitions [{
      sphereCollisionRadius=0
      boundBone {boneName=neck_02}
      boxExtents=<10 3 10>
      constraintSetup {angularLimitsMax=<3 10 3>, angularLimitsMin=<-3 -10 -3>, angularTarget=<0 0 0>}
      localJointOffset=<1 0 0>
     }]
   }
   n3 = Anim Dynamics - Bone: head(ComponentPose=@n2, Alpha=@n1, ExternalForce="0,0,0") {
    angularDampingOverride=0.999
    angularSpringConstant=200
    bGravityOverrideInSimSpace=true
    bOverrideAngularDamping=true
    lODThreshold=1
    numSolverIterationsPostUpdate=4
    numSolverIterationsPreUpdate=16
    alphaBoolBlend {blendOption=Linear}
    alphaScaleBiasClamp {bInterpResult=true, bMapRange=true, inRange {max=500}}
    boundBone {boneName=head}
    componentLinearAccScale=<0.5 0.5 0.5>
    physicsBodyDefinitions [{
      sphereCollisionRadius=0
      boundBone {boneName=head}
      boxExtents=<3 10 10>
      constraintSetup {angularTargetAxis=AxisY, twistAxis=AxisY, angularLimitsMax=<10 3 3>, angularLimitsMin=<-10 -3 -3>, angularTarget=<0 0 0>}
      localJointOffset=<0 1 0>
     }]
   }
   n4 = Anim Dynamics - Bone: neck_01(ComponentPose=Local To Component(LocalPose=Use cached pose 'Sub Graph Cache'()), Alpha=1, ExternalForce="0,0,0") {
    angularDampingOverride=0.95
    angularSpringConstant=500
    bAngularSpring=true
    bGravityOverrideInSimSpace=true
    bOverrideAngularDamping=true
    gravityScale=0.35
    lODThreshold=1
    numSolverIterationsPostUpdate=2
    numSolverIterationsPreUpdate=8
    alphaBoolBlend {blendOption=Linear}
    boundBone {boneName=neck_01}
    componentLinearAccScale=<0.8 0.8 0.8>
    componentLinearVelScale=<0.8 0.8 0.8>
    physicsBodyDefinitions [{
      sphereCollisionRadius=0
      boundBone {boneName=neck_01}
      boxExtents=<5 5 5>
      constraintSetup {angularTargetAxis=AxisY, twistAxis=AxisY, angularLimitsMax=<10 20 0>, angularLimitsMin=<-10 -20 -20>, angularTarget=<0 0 0>}
      localJointOffset=<0 0.2 0>
     }]
   }
   n5 = Anim Dynamics - Bone: head(ComponentPose=@n4, Alpha=1, ExternalForce="0,0,0") {
    angularDampingOverride=0.95
    angularSpringConstant=500
    bAngularSpring=true
    bGravityOverrideInSimSpace=true
    bOverrideAngularDamping=true
    gravityScale=0.35
    lODThreshold=1
    numSolverIterationsPostUpdate=2
    numSolverIterationsPreUpdate=8
    alphaBoolBlend {blendOption=Linear}
    boundBone {boneName=head}
    componentLinearAccScale=<0.8 0.8 0.8>
    componentLinearVelScale=<0.8 0.8 0.8>
    physicsBodyDefinitions [{
      sphereCollisionRadius=0
      boundBone {boneName=head}
      boxExtents=<5 5 5>
      constraintSetup {angularLimitsMax=<20 10 20>, angularLimitsMin=<-20 -10 0>, angularTarget=<0 0 0>}
      localJointOffset=<-0.2 0 0>
     }]
   }
   n6 = Blend Poses by bool(
    BlendPose_0=Component To Local(ComponentPose=@n5)
    BlendPose_1=Component To Local(ComponentPose=@n3)
    BlendTime_0=1
    BlendTime_1=1
    bActiveValue=Get bIsPlayingEmoteOrDBNO() -> (bIsPlayingEmoteOrDBNO:bool)
   ) {blendType=Linear}
   n7 = Use cached pose 'Sub Graph Cache'()
   n8 = Get bPetWantsAnimDynamics() -> (bPetWantsAnimDynamics:bool)
   n9 = "Constraint - tongue1 (P)"(
    ComponentPose=Local To Component(LocalPose=Blend Poses by bool(BlendPose_0=@n6, BlendPose_1=@n7, BlendTime_0=1, BlendTime_1=1, bActiveValue=@n8) {blendType=Linear})
    Alpha=1
    ConstraintWeights_0=1
   ) {
    alphaBoolBlend {blendOption=Linear}
    boneToModify {boneName=tongue1}
    constraintSetup [{targetBone {boneName=tongue}}]
    constraintWeights [1]
   }
   n10 = "Constraint - tongue2 (P)"(ComponentPose=@n9, Alpha=1, ConstraintWeights_0=1) {
    alphaBoolBlend {blendOption=Linear}
    boneToModify {boneName=tongue2}
    constraintSetup [{targetBone {boneName=tongue1}}]
    constraintWeights [1]
   }
   n11 = "Constraint - tongue3 (P)"(ComponentPose=@n10, Alpha=1, ConstraintWeights_0=1) {
    alphaBoolBlend {blendOption=Linear}
    boneToModify {boneName=tongue3}
    constraintSetup [{targetBone {boneName=tongue2}}]
    constraintWeights [1]
   }
   n12 = "Constraint - tongue4 (P)"(ComponentPose=@n11, Alpha=1, ConstraintWeights_0=1) {
    alphaBoolBlend {blendOption=Linear}
    boneToModify {boneName=tongue4}
    constraintSetup [{offsetOption=None, targetBone {boneName=tongue3}}]
    constraintWeights [1]
   }
   n13 = "Constraint - tongue1 (S)"(ComponentPose=@n12, Alpha=1, ConstraintWeights_0=1) {
    alphaBoolBlend {blendOption=Linear}
    boneToModify {boneName=tongue1}
    constraintSetup [{transformType=Scale, targetBone {boneName=tongue}}]
    constraintWeights [1]
   }
   n14 = "Constraint - tongue2 (S)"(ComponentPose=@n13, Alpha=1, ConstraintWeights_0=1) {
    alphaBoolBlend {blendOption=Linear}
    boneToModify {boneName=tongue2}
    constraintSetup [{offsetOption=None, transformType=Scale, targetBone {boneName=tongue1}}]
    constraintWeights [1]
   }
   n15 = "Constraint - tongue3 (S)"(ComponentPose=@n14, Alpha=1, ConstraintWeights_0=1) {
    alphaBoolBlend {blendOption=Linear}
    boneToModify {boneName=tongue3}
    constraintSetup [{offsetOption=None, transformType=Scale, targetBone {boneName=tongue2}}]
    constraintWeights [1]
   }
   n16 = "Constraint - tongue4 (S)"(ComponentPose=@n15, Alpha=1, ConstraintWeights_0=1) {
    alphaBoolBlend {blendOption=Linear}
    boneToModify {boneName=tongue4}
    constraintSetup [{offsetOption=None, transformType=Scale, targetBone {boneName=tongue3}}]
    constraintWeights [1]
   }
   Output Pose(Result=Component To Local(ComponentPose=@n16)) {layerGroup=None}
   AnimGraphNode_SubInput_1 = Input Pose()
   Save cached pose 'Sub Graph Cache'(Pose=@AnimGraphNode_SubInput_1) {cacheName=Sub Graph Cache}
  ]}
 EventGraph {links [], nodes []}
]
variables []
